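# Animation layer interface for AI fauna that declares overridable locomotion layers, with no target skeleton set.
# It declares two layers, Fauna Locomotion and Fauna LocomotionTurn, each taking its own input pose and returning an output pose.
# Both share the FaunaLocomotionOverride layer group, so linked implementations of the two run on one shared instance.
# The graphs hold only these signatures, with no animation logic or variables.
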AnimBlueprint /Irwin/AI/Interfaces/Fauna_Locomotion_AnimInterface
target_skeleton=null
graphs [
 Fauna Locomotion {nodes [
   Input Pose() {name=InPose Locomotion}
   Output Pose() {layerGroup=FaunaLocomotionOverride}
  ]}
 Fauna LocomotionTurn {nodes [
   Input Pose() {name=InPose LocomotionTurn}
   Output Pose() {layerGroup=FaunaLocomotionOverride}
  ]}
]
variables []
