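# Animation Layer Interface for fauna AI that defines overridable idle and idle-turn animation layers.
# It declares two layer functions, Fauna Idle and Fauna IdleTurn, each taking an input pose and returning an output pose.
# Both share the FaunaIdleOverride group, so implementations in that group run on one linked instance.
# It has no logic, variables or target skeleton, leaving the animation to implementing Anim Blueprints.
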AnimBlueprint /Irwin/AI/Interfaces/Fauna_Idle_AnimInterface
target_skeleton=null
graphs [
 Fauna Idle {nodes [
   Input Pose() {name=InPose Idle}
   Output Pose() {layerGroup=FaunaIdleOverride}
  ]}
 Fauna IdleTurn {nodes [
   Input Pose() {name=InPose IdleTurn}
   Output Pose() {layerGroup=FaunaIdleOverride}
  ]}
]
variables []
